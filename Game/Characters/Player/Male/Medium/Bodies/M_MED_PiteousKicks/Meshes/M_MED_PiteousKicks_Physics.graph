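# Physics asset for the medium male player body PiteousKicks, covering its lower body and dangling belt straps and sword.
# Capsule bodies on the pelvis, lower spine, thighs, calves and feet are joined by constraints with free swing and twist.
# Two four-segment strap chains hang from the pelvis and the sword from spine_01, all swing-limited with zero twist, while the body capsules provide shapes they can collide against.
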PhysicsAsset /Game/Characters/Player/Male/Medium/Bodies/M_MED_PiteousKicks/Meshes/M_MED_PiteousKicks_Physics
bodies {
 calf_l [{shapeType=Capsule, length=35.5, radius=7.5, shapeName=calf_l_capsule, center=<-13.5 -1.4457 0>, rotation=<rot -90 -0.0001 -0.0001>}]
 calf_r [{shapeType=Capsule, length=35.5, radius=7.5, shapeName=calf_r_capsule, center=<13.5 1.2145 0>, rotation=<rot 90 0 0.0005>}]
 dyn_belt_strap_l_01 [{shapeType=Capsule, length=6, radius=2.5, shapeName=dyn_belt_strap_l_01_capsule, center=<3.3362 0 0>, rotation=<rot 70 -89.9996 -89.9999>}]
 dyn_belt_strap_l_02 [{shapeType=Capsule, length=6, radius=2.5, shapeName=dyn_belt_strap_l_02_capsule, center=<3.3362 0 0>, rotation=<rot 70 -89.9996 -89.9999>}]
 dyn_belt_strap_l_03 [{shapeType=Capsule, length=5.5, radius=3, shapeName=dyn_belt_strap_l_03_capsule, center=<3.3362 0 0>, rotation=<rot 70 -89.9996 -89.9999>}]
 dyn_belt_strap_l_04 [{shapeType=Capsule, length=5.5, radius=3, shapeName=dyn_belt_strap_l_04_capsule, center=<3.3362 0 0>, rotation=<rot 70 -89.9996 -89.9999>}]
 dyn_belt_strap_r_01 [{shapeType=Capsule, length=6, radius=2.5, shapeName=dyn_belt_strap_r_01_capsule, center=<3.3362 0 0>, rotation=<rot 75 89.9991 89.9994>}]
 dyn_belt_strap_r_02 [{shapeType=Capsule, length=6, radius=2.5, shapeName=dyn_belt_strap_r_02_capsule, center=<3.3362 0 0>, rotation=<rot 75 89.9991 89.9994>}]
 dyn_belt_strap_r_03 [{shapeType=Capsule, length=5.5, radius=3, shapeName=dyn_belt_strap_r_03_capsule, center=<3.3362 0 0>, rotation=<rot 75 89.9991 89.9995>}]
 dyn_belt_strap_r_04 [{shapeType=Capsule, length=5.5, radius=3, shapeName=dyn_belt_strap_r_04_capsule, center=<3.3362 0 0>, rotation=<rot 75 89.9991 89.9995>}]
 dyn_sword [{
   shapeType=Capsule
   length=33.4044
   radius=3.595
   shapeName=dyn_sword_capsule
   center=<6.2492 -2.4844 -7.1279>
   rotation=<rot 21.8836 -90 -14.0025>
  }]
 foot_l [{shapeType=Capsule, length=15.5, radius=7.5, shapeName=foot_l_capsule, center=<-3.9093 6.3107 0>, rotation=<rot 0 -0.0002 90>}]
 foot_r [{shapeType=Capsule, length=15.5, radius=7.5, shapeName=foot_r_capsule, center=<3.9093 -6.3108 0>, rotation=<rot 0 179.9995 90>}]
 pelvis [{shapeType=Capsule, length=12, radius=10, shapeName=pelvis_capsule}]
 spine_01 [{
   shapeType=Capsule
   length=4.0298
   radius=10.9572
   shapeName=spine_01_capsule
   center=<2.3998 0.7664 0.0045>
   rotation=<rot 0.1972 -90 -0.0753>
  }]
 thigh_l [{shapeType=Capsule, length=26, radius=7.5, shapeName=thigh_l_capsule, center=<-17.2417 -0.411 0.0041>, rotation=<rot 90 180 -4.9988>}]
 thigh_r [{shapeType=Capsule, length=26, radius=7.5, shapeName=thigh_r_capsule, center=<17.2417 0.4111 -0.183>, rotation=<rot 90 179.6419 174.6427>}]
}
constraints [
 {
  bone1Name=calf_l
  bone2Name=thigh_l
  swing1LimitDegrees=45
  swing1Motion=Free
  swing2LimitDegrees=45
  swing2Motion=Free
  twistLimitDegrees=45
  twistMotion=Free
 }
 {
  bone1Name=calf_r
  bone2Name=thigh_r
  swing1LimitDegrees=45
  swing1Motion=Free
  swing2LimitDegrees=45
  swing2Motion=Free
  twistLimitDegrees=45
  twistMotion=Free
 }
 {bone1Name=dyn_belt_strap_l_02, bone2Name=dyn_belt_strap_l_01, swing1LimitDegrees=20, swing2LimitDegrees=30, twistLimitDegrees=0}
 {bone1Name=dyn_belt_strap_l_03, bone2Name=dyn_belt_strap_l_02, swing1LimitDegrees=20, swing2LimitDegrees=30, twistLimitDegrees=0}
 {bone1Name=dyn_belt_strap_l_04, bone2Name=dyn_belt_strap_l_03, swing1LimitDegrees=20, swing2LimitDegrees=30, twistLimitDegrees=0}
 {bone1Name=dyn_belt_strap_r_02, bone2Name=dyn_belt_strap_r_01, swing1LimitDegrees=20, swing2LimitDegrees=30, twistLimitDegrees=0}
 {bone1Name=dyn_belt_strap_r_03, bone2Name=dyn_belt_strap_r_02, swing1LimitDegrees=20, swing2LimitDegrees=30, twistLimitDegrees=0}
 {bone1Name=dyn_belt_strap_r_04, bone2Name=dyn_belt_strap_r_03, swing1LimitDegrees=20, swing2LimitDegrees=30, twistLimitDegrees=0}
 {
  bone1Name=thigh_l
  bone2Name=pelvis
  swing1LimitDegrees=45
  swing1Motion=Free
  swing2LimitDegrees=45
  swing2Motion=Free
  twistLimitDegrees=45
  twistMotion=Free
 }
 {
  bone1Name=thigh_r
  bone2Name=pelvis
  swing1LimitDegrees=45
  swing1Motion=Free
  swing2LimitDegrees=45
  swing2Motion=Free
  twistLimitDegrees=45
  twistMotion=Free
 }
 {bone1Name=dyn_belt_strap_r_01, bone2Name=pelvis, swing1LimitDegrees=20, swing2LimitDegrees=30, twistLimitDegrees=0}
 {bone1Name=dyn_belt_strap_l_01, bone2Name=pelvis, swing1LimitDegrees=20, swing2LimitDegrees=30, twistLimitDegrees=0}
 {
  bone1Name=foot_r
  bone2Name=calf_r
  swing1LimitDegrees=45
  swing1Motion=Free
  swing2LimitDegrees=45
  swing2Motion=Free
  twistLimitDegrees=45
  twistMotion=Free
 }
 {
  bone1Name=foot_l
  bone2Name=calf_l
  swing1LimitDegrees=45
  swing1Motion=Free
  swing2LimitDegrees=45
  swing2Motion=Free
  twistLimitDegrees=45
  twistMotion=Free
 }
 {bone1Name=dyn_sword, bone2Name=spine_01, swing1LimitDegrees=60, swing2LimitDegrees=60, twistLimitDegrees=0}
 {
  bone1Name=spine_01
  bone2Name=pelvis
  swing1LimitDegrees=45
  swing1Motion=Free
  swing2LimitDegrees=45
  swing2Motion=Free
  twistLimitDegrees=45
  twistMotion=Free
 }
]
