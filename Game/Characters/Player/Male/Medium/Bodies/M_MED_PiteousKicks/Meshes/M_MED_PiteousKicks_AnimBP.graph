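# Post-process-style AnimBlueprint for the PiteousKicks medium male player body, adding deform-rig correction and secondary motion to the pose copied from the parent mesh.
# Control Rig feeds weighted spine/clavicle constraints on pistol bases, RigidBody physics, then AnimDynamics for belt flaps and left shoulder-pad chain.
# Each AnimDynamics bone is simulated twice, bone-relative then world space.
# Updates set gravity override, backbling-dependent strap acceleration scale, and front-end/in-game dynamics values.
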AnimBlueprint /Game/Characters/Player/Male/Medium/Bodies/M_MED_PiteousKicks/Meshes/M_MED_PiteousKicks_AnimBP
target_skeleton=&/M_MED_PiteousKicks_Skeleton
graphs [
 AnimGraph {nodes [
   n1 = Control Rig(Source=Copy Pose From Mesh() {bUseAttachedParent=true}) {
    alphaBoolBlend {blendOption=Linear}
    controlRigAssetReference {blueprintRigClass=&/M_MED_PiteousKicks_DeformRig_ControlRig.M_MED_PiteousKicks_DeformRig_ControlRig_C}
   }
   n2 = "Constraint - dyn_pistol_l_base (PPPPP)"!(
    ComponentPose=Local To Component(LocalPose=@n1)
    Alpha=1
    ConstraintWeights_0=0.6089
    ConstraintWeights_1:float=0.2201
    ConstraintWeights_2:float=0.0892
    ConstraintWeights_3:float=0.0658
    ConstraintWeights_4:float=0.0161
   ) -> (Pose) {
    alphaBoolBlend {blendOption=Linear}
    boneToModify {boneName=dyn_pistol_l_base}
    constraintSetup [
     {targetBone {boneName=spine_04}}
     {targetBone {boneName=clavicle_l}}
     {targetBone {boneName=spine_03}}
     {targetBone {boneName=spine_05}}
     {targetBone {boneName=spine_02}}
    ]
    constraintWeights [1,1,1,1,1]
   }
   n3 = "Constraint - dyn_pistol_r_base (PPPPP)"!(
    ComponentPose=@n2
    Alpha=1
    ConstraintWeights_0=0.6067
    ConstraintWeights_1:float=0.251
    ConstraintWeights_2:float=0.0835
    ConstraintWeights_3:float=0.048
    ConstraintWeights_4:float=0.0107
   ) -> (Pose) {
    alphaBoolBlend {blendOption=Linear}
    boneToModify {boneName=dyn_pistol_r_base}
    constraintSetup [
     {targetBone {boneName=spine_04}}
     {targetBone {boneName=clavicle_r}}
     {targetBone {boneName=spine_03}}
     {targetBone {boneName=spine_05}}
     {targetBone {boneName=spine_02}}
    ]
    constraintWeights [1,1,1,1,1]
   }
   n4 = RigidBody!(ComponentPose=@n3, Alpha, ExternalForce, ComponentLinearAccScale:Vector="0,0,0") -> (Pose) {
    bClampLinearTranslationLimitToRefPose=true
    bOverrideWorldGravity=true
    overridePhysicsAsset=&/M_MED_PiteousKicks_Physics
    simulationSpace=BaseBoneSpace
    alphaBoolBlend {blendOption=Linear}
    baseBoneRef {boneName=pelvis}
    componentAppliedLinearAccClamp=<3000 3000 3000>
    componentLinearAccScale=<0.3 0.3 0.3>
    componentLinearVelScale=<0.3 0.3 0.3>
    overrideWorldGravity {z=-980}
   }
   n5 = Anim Dynamics - Bone: dyn_belt_flap_fr_l_01(ComponentPose=@n4, Alpha=1, ExternalForce="0,0,0") {
    angularDampingOverride=0.9
    angularSpringConstant=0.4
    bAngularSpring=true
    bGravityOverrideInSimSpace=true
    bUseGravityOverride=true
    simulationSpace=BoneRelative
    alphaBoolBlend {blendOption=Linear}
    boundBone {boneName=dyn_belt_flap_fr_l_01}
    componentAppliedLinearAccClamp=<10000 10000 10000>
    componentLinearAccScale=<1 1 1>
    componentLinearVelScale=<1 1 1>
    gravityOverride {z=-1200}
    physicsBodyDefinitions [{
      boundBone {boneName=dyn_belt_flap_fr_l_01}
      boxExtents=<21 21 21>
      constraintSetup {angularLimitsMax=<0 2 0>, angularLimitsMin=<0 -20 0>, angularTarget=<0 0 0>}
      localJointOffset=<1 0 0>
     }]
    relativeSpaceBone {boneName=root}
   }
   n6 = Anim Dynamics - Bone: dyn_belt_flap_fr_l_02(ComponentPose=@n5, Alpha=1, ExternalForce="0,0,0") {
    angularDampingOverride=0.9
    angularSpringConstant=0.4
    bAngularSpring=true
    bGravityOverrideInSimSpace=true
    bUseGravityOverride=true
    simulationSpace=BoneRelative
    alphaBoolBlend {blendOption=Linear}
    boundBone {boneName=dyn_belt_flap_fr_l_02}
    componentAppliedLinearAccClamp=<10000 10000 10000>
    componentLinearAccScale=<1 1 1>
    componentLinearVelScale=<1 1 1>
    gravityOverride {z=-1200}
    physicsBodyDefinitions [{
      boundBone {boneName=dyn_belt_flap_fr_l_02}
      boxExtents=<21 21 21>
      constraintSetup {angularLimitsMax=<0 2 0>, angularLimitsMin=<0 -20 0>, angularTarget=<0 0 0>}
      localJointOffset=<1 0 0>
     }]
    relativeSpaceBone {boneName=root}
   }
   n7 = Anim Dynamics - Bone: dyn_belt_flap_fr_l_03(ComponentPose=@n6, Alpha=1, ExternalForce="0,0,0") {
    angularDampingOverride=0.9
    angularSpringConstant=0.4
    bAngularSpring=true
    bGravityOverrideInSimSpace=true
    bUseGravityOverride=true
    simulationSpace=BoneRelative
    alphaBoolBlend {blendOption=Linear}
    boundBone {boneName=dyn_belt_flap_fr_l_03}
    componentAppliedLinearAccClamp=<10000 10000 10000>
    componentLinearAccScale=<1 1 1>
    componentLinearVelScale=<1 1 1>
    gravityOverride {z=-1200}
    physicsBodyDefinitions [{
      boundBone {boneName=dyn_belt_flap_fr_l_03}
      boxExtents=<23 23 23>
      constraintSetup {angularLimitsMax=<0 2 0>, angularLimitsMin=<0 -20 0>, angularTarget=<0 0 0>}
      localJointOffset=<1 0 0>
     }]
    relativeSpaceBone {boneName=root}
   }
   n8 = Anim Dynamics - Bone: dyn_belt_flap_fr_l_04(ComponentPose=@n7, Alpha=1, ExternalForce="0,0,0") {
    angularDampingOverride=0.9
    angularSpringConstant=0.4
    bAngularSpring=true
    bGravityOverrideInSimSpace=true
    bUseGravityOverride=true
    simulationSpace=BoneRelative
    alphaBoolBlend {blendOption=Linear}
    boundBone {boneName=dyn_belt_flap_fr_l_04}
    componentAppliedLinearAccClamp=<10000 10000 10000>
    componentLinearAccScale=<1 1 1>
    componentLinearVelScale=<1 1 1>
    gravityOverride {z=-1200}
    physicsBodyDefinitions [{
      boundBone {boneName=dyn_belt_flap_fr_l_04}
      boxExtents=<23 23 23>
      constraintSetup {angularLimitsMax=<0 2 0>, angularLimitsMin=<0 -20 0>, angularTarget=<0 0 0>}
      localJointOffset=<1 0 0>
     }]
    relativeSpaceBone {boneName=root}
   }
   n9 = Anim Dynamics - Bone: dyn_belt_flap_fr_l_05(ComponentPose=@n8, Alpha=1, ExternalForce="0,0,0") {
    angularDampingOverride=0.9
    angularSpringConstant=0.4
    bAngularSpring=true
    bGravityOverrideInSimSpace=true
    bUseGravityOverride=true
    simulationSpace=BoneRelative
    alphaBoolBlend {blendOption=Linear}
    boundBone {boneName=dyn_belt_flap_fr_l_05}
    componentAppliedLinearAccClamp=<10000 10000 10000>
    componentLinearAccScale=<1 1 1>
    componentLinearVelScale=<1 1 1>
    gravityOverride {z=-1200}
    physicsBodyDefinitions [{
      boundBone {boneName=dyn_belt_flap_fr_l_05}
      boxExtents=<21 21 21>
      constraintSetup {angularLimitsMax=<0 2 0>, angularLimitsMin=<0 -20 0>, angularTarget=<0 0 0>}
      localJointOffset=<1 0 0>
     }]
    relativeSpaceBone {boneName=root}
   }
   n10 = Anim Dynamics - Bone: dyn_belt_flap_fr_r_01(ComponentPose=@n9, Alpha=1, ExternalForce="0,0,0") {
    angularDampingOverride=0.9
    angularSpringConstant=0.4
    bGravityOverrideInSimSpace=true
    bOverrideAngularDamping=true
    bUseGravityOverride=true
    simulationSpace=BoneRelative
    alphaBoolBlend {blendOption=Linear}
    boundBone {boneName=dyn_belt_flap_fr_r_01}
    componentAppliedLinearAccClamp=<10000 10000 10000>
    componentLinearAccScale=<1 1 1>
    componentLinearVelScale=<1 1 1>
    gravityOverride {z=-1200}
    physicsBodyDefinitions [{
      boundBone {boneName=dyn_belt_flap_fr_r_01}
      boxExtents=<21 21 21>
      constraintSetup {angularLimitsMax=<0 2 0>, angularLimitsMin=<0 -20 0>, angularTarget=<0 0 0>}
      localJointOffset=<1 0 0>
     }]
    relativeSpaceBone {boneName=root}
   }
   n11 = Anim Dynamics - Bone: dyn_belt_flap_fr_r_02(ComponentPose=@n10, Alpha=1, ExternalForce="0,0,0") {
    angularDampingOverride=0.9
    angularSpringConstant=0.4
    bGravityOverrideInSimSpace=true
    bOverrideAngularDamping=true
    bUseGravityOverride=true
    simulationSpace=BoneRelative
    alphaBoolBlend {blendOption=Linear}
    boundBone {boneName=dyn_belt_flap_fr_r_02}
    componentAppliedLinearAccClamp=<10000 10000 10000>
    componentLinearAccScale=<1 1 1>
    componentLinearVelScale=<1 1 1>
    gravityOverride {z=-1200}
    physicsBodyDefinitions [{
      boundBone {boneName=dyn_belt_flap_fr_r_02}
      boxExtents=<21 21 21>
      constraintSetup {angularLimitsMax=<0 2 0>, angularLimitsMin=<0 -20 0>, angularTarget=<0 0 0>}
      localJointOffset=<1 0 0>
     }]
    relativeSpaceBone {boneName=root}
   }
   n12 = Anim Dynamics - Bone: dyn_belt_flap_fr_r_03(ComponentPose=@n11, Alpha=1, ExternalForce="0,0,0") {
    angularDampingOverride=0.9
    angularSpringConstant=0.4
    bGravityOverrideInSimSpace=true
    bOverrideAngularDamping=true
    bUseGravityOverride=true
    simulationSpace=BoneRelative
    alphaBoolBlend {blendOption=Linear}
    boundBone {boneName=dyn_belt_flap_fr_r_03}
    componentAppliedLinearAccClamp=<10000 10000 10000>
    componentLinearAccScale=<1 1 1>
    componentLinearVelScale=<1 1 1>
    gravityOverride {z=-1200}
    physicsBodyDefinitions [{
      boundBone {boneName=dyn_belt_flap_fr_r_03}
      boxExtents=<23 23 23>
      constraintSetup {angularLimitsMax=<0 2 0>, angularLimitsMin=<0 -20 0>, angularTarget=<0 0 0>}
      localJointOffset=<1 0 0>
     }]
    relativeSpaceBone {boneName=root}
   }
   n13 = Anim Dynamics - Bone: dyn_belt_flap_fr_r_04(ComponentPose=@n12, Alpha=1, ExternalForce="0,0,0") {
    angularDampingOverride=0.9
    angularSpringConstant=0.4
    bGravityOverrideInSimSpace=true
    bOverrideAngularDamping=true
    bUseGravityOverride=true
    simulationSpace=BoneRelative
    alphaBoolBlend {blendOption=Linear}
    boundBone {boneName=dyn_belt_flap_fr_r_04}
    componentAppliedLinearAccClamp=<10000 10000 10000>
    componentLinearAccScale=<1 1 1>
    componentLinearVelScale=<1 1 1>
    gravityOverride {z=-1200}
    physicsBodyDefinitions [{
      boundBone {boneName=dyn_belt_flap_fr_r_04}
      boxExtents=<23 23 23>
      constraintSetup {angularLimitsMax=<0 2 0>, angularLimitsMin=<0 -20 0>, angularTarget=<0 0 0>}
      localJointOffset=<1 0 0>
     }]
    relativeSpaceBone {boneName=root}
   }
   n14 = Anim Dynamics - Bone: dyn_belt_flap_fr_r_05(ComponentPose=@n13, Alpha=1, ExternalForce="0,0,0") {
    angularDampingOverride=0.9
    angularSpringConstant=0.4
    bGravityOverrideInSimSpace=true
    bOverrideAngularDamping=true
    bUseGravityOverride=true
    simulationSpace=BoneRelative
    alphaBoolBlend {blendOption=Linear}
    boundBone {boneName=dyn_belt_flap_fr_r_05}
    componentAppliedLinearAccClamp=<10000 10000 10000>
    componentLinearAccScale=<1 1 1>
    componentLinearVelScale=<1 1 1>
    gravityOverride {z=-1200}
    physicsBodyDefinitions [{
      boundBone {boneName=dyn_belt_flap_fr_r_05}
      boxExtents=<21 21 21>
      constraintSetup {angularLimitsMax=<0 2 0>, angularLimitsMin=<0 -20 0>, angularTarget=<0 0 0>}
      localJointOffset=<1 0 0>
     }]
    relativeSpaceBone {boneName=root}
   }
   n15 = Anim Dynamics - Bone: dyn_belt_flap_bck_l_01(ComponentPose=@n14, Alpha=1, ExternalForce="0,0,0") {
    angularSpringConstant=0.5
    bAngularSpring=true
    bGravityOverrideInSimSpace=true
    bUseGravityOverride=true
    simulationSpace=BoneRelative
    alphaBoolBlend {blendOption=Linear}
    boundBone {boneName=dyn_belt_flap_bck_l_01}
    componentAppliedLinearAccClamp=<10000 10000 10000>
    componentLinearAccScale=<1 1 1>
    componentLinearVelScale=<1 1 1>
    gravityOverride {z=-1200}
    physicsBodyDefinitions [{
      boundBone {boneName=dyn_belt_flap_bck_l_01}
      boxExtents=<23 23 23>
      constraintSetup {angularLimitsMax=<0 0 1>, angularLimitsMin=<0 -50 -1>, angularTarget=<0 0 0>}
      localJointOffset=<1 0 0>
     }]
    relativeSpaceBone {boneName=root}
   }
   n16 = Anim Dynamics - Bone: dyn_belt_flap_bck_l_02(ComponentPose=@n15, Alpha=1, ExternalForce="0,0,0") {
    angularSpringConstant=0.3
    bAngularSpring=true
    bGravityOverrideInSimSpace=true
    bUseGravityOverride=true
    simulationSpace=BoneRelative
    alphaBoolBlend {blendOption=Linear}
    boundBone {boneName=dyn_belt_flap_bck_l_02}
    componentAppliedLinearAccClamp=<10000 10000 10000>
    componentLinearAccScale=<1 1 1>
    componentLinearVelScale=<1 1 1>
    gravityOverride {z=-1200}
    physicsBodyDefinitions [{
      boundBone {boneName=dyn_belt_flap_bck_l_02}
      boxExtents=<23 23 23>
      constraintSetup {angularLimitsMax=<0 2 1>, angularLimitsMin=<0 -50 -1>, angularTarget=<0 0 0>}
      localJointOffset=<1 0 0>
     }]
    relativeSpaceBone {boneName=root}
   }
   n17 = Anim Dynamics - Bone: dyn_belt_flap_bck_l_03(ComponentPose=@n16, Alpha=1, ExternalForce="0,0,0") {
    angularSpringConstant=0.3
    bAngularSpring=true
    bGravityOverrideInSimSpace=true
    bUseGravityOverride=true
    simulationSpace=BoneRelative
    alphaBoolBlend {blendOption=Linear}
    boundBone {boneName=dyn_belt_flap_bck_l_03}
    componentAppliedLinearAccClamp=<10000 10000 10000>
    componentLinearAccScale=<1 1 1>
    componentLinearVelScale=<1 1 1>
    gravityOverride {z=-1200}
    physicsBodyDefinitions [{
      boundBone {boneName=dyn_belt_flap_bck_l_03}
      boxExtents=<21 21 21>
      constraintSetup {angularLimitsMax=<0 0 1>, angularLimitsMin=<0 -50 -1>, angularTarget=<0 0 0>}
      localJointOffset=<1 0 0>
     }]
    relativeSpaceBone {boneName=root}
   }
   n18 = Anim Dynamics - Bone: dyn_belt_flap_bck_l_04(ComponentPose=@n17, Alpha=1, ExternalForce="0,0,0") {
    angularSpringConstant=0.5
    bAngularSpring=true
    bGravityOverrideInSimSpace=true
    bUseGravityOverride=true
    simulationSpace=BoneRelative
    alphaBoolBlend {blendOption=Linear}
    boundBone {boneName=dyn_belt_flap_bck_l_04}
    componentAppliedLinearAccClamp=<10000 10000 10000>
    componentLinearAccScale=<1 1 1>
    componentLinearVelScale=<1 1 1>
    gravityOverride {z=-1200}
    physicsBodyDefinitions [{
      boundBone {boneName=dyn_belt_flap_bck_l_04}
      boxExtents=<21 21 21>
      constraintSetup {angularLimitsMax=<0 2 1>, angularLimitsMin=<0 -50 -1>, angularTarget=<0 0 0>}
      localJointOffset=<1 0 0>
     }]
    relativeSpaceBone {boneName=root}
   }
   n19 = Anim Dynamics - Bone: dyn_belt_flap_bck_r_01(ComponentPose=@n18, Alpha=1, ExternalForce="0,0,0") {
    angularSpringConstant=5.2
    bAngularSpring=true
    bGravityOverrideInSimSpace=true
    bUseGravityOverride=true
    simulationSpace=BoneRelative
    alphaBoolBlend {blendOption=Linear}
    boundBone {boneName=dyn_belt_flap_bck_r_01}
    componentAppliedLinearAccClamp=<10000 10000 10000>
    componentLinearAccScale=<1 1 1>
    componentLinearVelScale=<1 1 1>
    gravityOverride {z=-1200}
    physicsBodyDefinitions [{
      boundBone {boneName=dyn_belt_flap_bck_r_01}
      boxExtents=<23 23 23>
      constraintSetup {angularLimitsMax=<0 0 1>, angularLimitsMin=<0 -50 -1>, angularTarget=<0 0 0>}
      localJointOffset=<1 0 0>
     }]
    relativeSpaceBone {boneName=root}
   }
   n20 = Anim Dynamics - Bone: dyn_belt_flap_bck_r_02(ComponentPose=@n19, Alpha=1, ExternalForce="0,0,0") {
    angularSpringConstant=0.3
    bAngularSpring=true
    bGravityOverrideInSimSpace=true
    bUseGravityOverride=true
    simulationSpace=BoneRelative
    alphaBoolBlend {blendOption=Linear}
    boundBone {boneName=dyn_belt_flap_bck_r_02}
    componentAppliedLinearAccClamp=<10000 10000 10000>
    componentLinearAccScale=<1 1 1>
    componentLinearVelScale=<1 1 1>
    gravityOverride {z=-1200}
    physicsBodyDefinitions [{
      boundBone {boneName=dyn_belt_flap_bck_r_02}
      boxExtents=<23 23 23>
      constraintSetup {angularLimitsMax=<0 2 1>, angularLimitsMin=<0 -50 -1>, angularTarget=<0 0 0>}
      localJointOffset=<1 0 0>
     }]
    relativeSpaceBone {boneName=root}
   }
   n21 = Anim Dynamics - Bone: dyn_belt_flap_bck_r_03(ComponentPose=@n20, Alpha=1, ExternalForce="0,0,0") {
    angularSpringConstant=0.3
    bAngularSpring=true
    bGravityOverrideInSimSpace=true
    bUseGravityOverride=true
    simulationSpace=BoneRelative
    alphaBoolBlend {blendOption=Linear}
    boundBone {boneName=dyn_belt_flap_bck_r_03}
    componentAppliedLinearAccClamp=<10000 10000 10000>
    componentLinearAccScale=<1 1 1>
    componentLinearVelScale=<1 1 1>
    gravityOverride {z=-1200}
    physicsBodyDefinitions [{
      boundBone {boneName=dyn_belt_flap_bck_r_03}
      boxExtents=<21 21 21>
      constraintSetup {angularLimitsMax=<0 0 1>, angularLimitsMin=<0 -50 -1>, angularTarget=<0 0 0>}
      localJointOffset=<1 0 0>
     }]
    relativeSpaceBone {boneName=root}
   }
   n22 = Anim Dynamics - Bone: dyn_belt_flap_bck_r_04(ComponentPose=@n21, Alpha=1, ExternalForce="0,0,0") {
    angularSpringConstant=0.5
    bAngularSpring=true
    bGravityOverrideInSimSpace=true
    bUseGravityOverride=true
    simulationSpace=BoneRelative
    alphaBoolBlend {blendOption=Linear}
    boundBone {boneName=dyn_belt_flap_bck_r_04}
    componentAppliedLinearAccClamp=<10000 10000 10000>
    componentLinearAccScale=<1 1 1>
    componentLinearVelScale=<1 1 1>
    gravityOverride {z=-1200}
    physicsBodyDefinitions [{
      boundBone {boneName=dyn_belt_flap_bck_r_04}
      boxExtents=<21 21 21>
      constraintSetup {angularLimitsMax=<0 2 1>, angularLimitsMin=<0 -50 -1>, angularTarget=<0 0 0>}
      localJointOffset=<1 0 0>
     }]
    relativeSpaceBone {boneName=root}
   }
   n23 = Anim Dynamics - Bone: dyn_belt_flap_fr_l_01(ComponentPose=@n22, Alpha=1, ExternalForce="0,0,0") {
    angularDampingOverride=0.9
    angularSpringConstant=5000
    bAngularSpring=true
    bGravityOverrideInSimSpace=true
    bUseGravityOverride=true
    simulationSpace=World
    alphaBoolBlend {blendOption=Linear}
    boundBone {boneName=dyn_belt_flap_fr_l_01}
    componentAppliedLinearAccClamp=<10000 10000 10000>
    componentLinearAccScale=<0.5 0.5 0.5>
    componentLinearVelScale=<0.5 0.5 0.5>
    gravityOverride {z=-1200}
    physicsBodyDefinitions [{
      boundBone {boneName=dyn_belt_flap_fr_l_01}
      boxExtents=<16 16 16>
      constraintSetup {angularLimitsMax=<0 2 0>, angularLimitsMin=<0 -20 0>, angularTarget=<0 0 0>}
      localJointOffset=<1 0 0>
     }]
    relativeSpaceBone {boneName=root}
   }
   n24 = Anim Dynamics - Bone: dyn_belt_flap_fr_l_02(ComponentPose=@n23, Alpha=1, ExternalForce="0,0,0") {
    angularDampingOverride=0.9
    angularSpringConstant=5000
    bAngularSpring=true
    bGravityOverrideInSimSpace=true
    bUseGravityOverride=true
    simulationSpace=World
    alphaBoolBlend {blendOption=Linear}
    boundBone {boneName=dyn_belt_flap_fr_l_02}
    componentAppliedLinearAccClamp=<10000 10000 10000>
    componentLinearAccScale=<0.5 0.5 0.5>
    componentLinearVelScale=<0.5 0.5 0.5>
    gravityOverride {z=-1200}
    physicsBodyDefinitions [{
      boundBone {boneName=dyn_belt_flap_fr_l_02}
      boxExtents=<16 16 16>
      constraintSetup {angularLimitsMax=<0 2 0>, angularLimitsMin=<0 -20 0>, angularTarget=<0 0 0>}
      localJointOffset=<1 0 0>
     }]
    relativeSpaceBone {boneName=root}
   }
   n25 = Anim Dynamics - Bone: dyn_belt_flap_fr_l_03(ComponentPose=@n24, Alpha=1, ExternalForce="0,0,0") {
    angularDampingOverride=0.9
    angularSpringConstant=3000
    bAngularSpring=true
    bGravityOverrideInSimSpace=true
    bUseGravityOverride=true
    simulationSpace=World
    alphaBoolBlend {blendOption=Linear}
    boundBone {boneName=dyn_belt_flap_fr_l_03}
    componentAppliedLinearAccClamp=<10000 10000 10000>
    componentLinearAccScale=<0.5 0.5 0.5>
    componentLinearVelScale=<0.5 0.5 0.5>
    gravityOverride {z=-1200}
    physicsBodyDefinitions [{
      boundBone {boneName=dyn_belt_flap_fr_l_03}
      boxExtents=<14 14 14>
      constraintSetup {angularLimitsMax=<0 2 0>, angularLimitsMin=<0 -20 0>, angularTarget=<0 0 0>}
      localJointOffset=<1 0 0>
     }]
    relativeSpaceBone {boneName=root}
   }
   n26 = Anim Dynamics - Bone: dyn_belt_flap_fr_l_04(ComponentPose=@n25, Alpha=1, ExternalForce="0,0,0") {
    angularDampingOverride=0.9
    angularSpringConstant=3000
    bAngularSpring=true
    bGravityOverrideInSimSpace=true
    bUseGravityOverride=true
    simulationSpace=World
    alphaBoolBlend {blendOption=Linear}
    boundBone {boneName=dyn_belt_flap_fr_l_04}
    componentAppliedLinearAccClamp=<10000 10000 10000>
    componentLinearAccScale=<0.5 0.5 0.5>
    componentLinearVelScale=<0.5 0.5 0.5>
    gravityOverride {z=-1200}
    physicsBodyDefinitions [{
      boundBone {boneName=dyn_belt_flap_fr_l_04}
      boxExtents=<14 14 14>
      constraintSetup {angularLimitsMax=<0 2 0>, angularLimitsMin=<0 -20 0>, angularTarget=<0 0 0>}
      localJointOffset=<1 0 0>
     }]
    relativeSpaceBone {boneName=root}
   }
   n27 = Anim Dynamics - Bone: dyn_belt_flap_fr_l_05(ComponentPose=@n26, Alpha=1, ExternalForce="0,0,0") {
    angularDampingOverride=0.9
    angularSpringConstant=5000
    bAngularSpring=true
    bGravityOverrideInSimSpace=true
    bUseGravityOverride=true
    simulationSpace=World
    alphaBoolBlend {blendOption=Linear}
    boundBone {boneName=dyn_belt_flap_fr_l_05}
    componentAppliedLinearAccClamp=<10000 10000 10000>
    componentLinearAccScale=<0.5 0.5 0.5>
    componentLinearVelScale=<0.5 0.5 0.5>
    gravityOverride {z=-1200}
    physicsBodyDefinitions [{
      boundBone {boneName=dyn_belt_flap_fr_l_05}
      boxExtents=<16 16 16>
      constraintSetup {angularLimitsMax=<0 2 0>, angularLimitsMin=<0 -20 0>, angularTarget=<0 0 0>}
      localJointOffset=<1 0 0>
     }]
    relativeSpaceBone {boneName=root}
   }
   n28 = Anim Dynamics - Bone: dyn_belt_flap_fr_r_01(ComponentPose=@n27, Alpha=1, ExternalForce="0,0,0") {
    angularDampingOverride=0.9
    angularSpringConstant=5000
    bAngularSpring=true
    bGravityOverrideInSimSpace=true
    bOverrideAngularDamping=true
    bUseGravityOverride=true
    simulationSpace=World
    alphaBoolBlend {blendOption=Linear}
    boundBone {boneName=dyn_belt_flap_fr_r_01}
    componentAppliedLinearAccClamp=<10000 10000 10000>
    componentLinearAccScale=<0.5 0.5 0.5>
    componentLinearVelScale=<0.5 0.5 0.5>
    gravityOverride {z=-1500}
    physicsBodyDefinitions [{
      boundBone {boneName=dyn_belt_flap_fr_r_01}
      boxExtents=<16 16 16>
      constraintSetup {angularLimitsMax=<0 2 0>, angularLimitsMin=<0 -20 0>, angularTarget=<0 0 0>}
      localJointOffset=<1 0 0>
     }]
    relativeSpaceBone {boneName=root}
   }
   n29 = Anim Dynamics - Bone: dyn_belt_flap_fr_r_02(ComponentPose=@n28, Alpha=1, ExternalForce="0,0,0") {
    angularDampingOverride=0.9
    angularSpringConstant=5000
    bAngularSpring=true
    bGravityOverrideInSimSpace=true
    bOverrideAngularDamping=true
    bUseGravityOverride=true
    simulationSpace=World
    alphaBoolBlend {blendOption=Linear}
    boundBone {boneName=dyn_belt_flap_fr_r_02}
    componentAppliedLinearAccClamp=<10000 10000 10000>
    componentLinearAccScale=<0.5 0.5 0.5>
    componentLinearVelScale=<0.5 0.5 0.5>
    gravityOverride {z=-1500}
    physicsBodyDefinitions [{
      boundBone {boneName=dyn_belt_flap_fr_r_02}
      boxExtents=<16 16 16>
      constraintSetup {angularLimitsMax=<0 2 0>, angularLimitsMin=<0 -20 0>, angularTarget=<0 0 0>}
      localJointOffset=<1 0 0>
     }]
    relativeSpaceBone {boneName=root}
   }
   n30 = Anim Dynamics - Bone: dyn_belt_flap_fr_r_03(ComponentPose=@n29, Alpha=1, ExternalForce="0,0,0") {
    angularDampingOverride=0.9
    angularSpringConstant=3000
    bAngularSpring=true
    bGravityOverrideInSimSpace=true
    bOverrideAngularDamping=true
    bUseGravityOverride=true
    simulationSpace=World
    alphaBoolBlend {blendOption=Linear}
    boundBone {boneName=dyn_belt_flap_fr_r_03}
    componentAppliedLinearAccClamp=<10000 10000 10000>
    componentLinearAccScale=<0.5 0.5 0.5>
    componentLinearVelScale=<0.5 0.5 0.5>
    gravityOverride {z=-1500}
    physicsBodyDefinitions [{
      boundBone {boneName=dyn_belt_flap_fr_r_03}
      boxExtents=<14 14 14>
      constraintSetup {angularLimitsMax=<0 2 0>, angularLimitsMin=<0 -20 0>, angularTarget=<0 0 0>}
      localJointOffset=<1 0 0>
     }]
    relativeSpaceBone {boneName=root}
   }
   n31 = Anim Dynamics - Bone: dyn_belt_flap_fr_r_04(ComponentPose=@n30, Alpha=1, ExternalForce="0,0,0") {
    angularDampingOverride=0.9
    angularSpringConstant=3000
    bAngularSpring=true
    bGravityOverrideInSimSpace=true
    bOverrideAngularDamping=true
    bUseGravityOverride=true
    simulationSpace=World
    alphaBoolBlend {blendOption=Linear}
    boundBone {boneName=dyn_belt_flap_fr_r_04}
    componentAppliedLinearAccClamp=<10000 10000 10000>
    componentLinearAccScale=<0.5 0.5 0.5>
    componentLinearVelScale=<0.5 0.5 0.5>
    gravityOverride {z=-1500}
    physicsBodyDefinitions [{
      boundBone {boneName=dyn_belt_flap_fr_r_04}
      boxExtents=<14 14 14>
      constraintSetup {angularLimitsMax=<0 2 0>, angularLimitsMin=<0 -20 0>, angularTarget=<0 0 0>}
      localJointOffset=<1 0 0>
     }]
    relativeSpaceBone {boneName=root}
   }
   n32 = Anim Dynamics - Bone: dyn_belt_flap_fr_r_05(ComponentPose=@n31, Alpha=1, ExternalForce="0,0,0") {
    angularDampingOverride=0.9
    angularSpringConstant=5000
    bAngularSpring=true
    bGravityOverrideInSimSpace=true
    bOverrideAngularDamping=true
    bUseGravityOverride=true
    simulationSpace=World
    alphaBoolBlend {blendOption=Linear}
    boundBone {boneName=dyn_belt_flap_fr_r_05}
    componentAppliedLinearAccClamp=<10000 10000 10000>
    componentLinearAccScale=<0.5 0.5 0.5>
    componentLinearVelScale=<0.5 0.5 0.5>
    gravityOverride {z=-1500}
    physicsBodyDefinitions [{
      boundBone {boneName=dyn_belt_flap_fr_r_05}
      boxExtents=<16 16 16>
      constraintSetup {angularLimitsMax=<0 2 0>, angularLimitsMin=<0 -20 0>, angularTarget=<0 0 0>}
      localJointOffset=<1 0 0>
     }]
    relativeSpaceBone {boneName=root}
   }
   n33 = Anim Dynamics - Bone: dyn_belt_flap_bck_l_01(ComponentPose=@n32, Alpha=1, ExternalForce="0,0,0") {
    angularSpringConstant=0.5
    bAngularSpring=true
    bGravityOverrideInSimSpace=true
    bUseGravityOverride=true
    simulationSpace=World
    alphaBoolBlend {blendOption=Linear}
    boundBone {boneName=dyn_belt_flap_bck_l_01}
    componentAppliedLinearAccClamp=<10000 10000 10000>
    componentLinearAccScale=<0.5 0.5 0.5>
    componentLinearVelScale=<0.5 0.5 0.5>
    gravityOverride {z=-1500}
    physicsBodyDefinitions [{
      boundBone {boneName=dyn_belt_flap_bck_l_01}
      boxExtents=<16 16 16>
      constraintSetup {angularLimitsMax=<0 0 1>, angularLimitsMin=<0 -50 -1>, angularTarget=<0 0 0>}
      localJointOffset=<1 0 0>
     }]
    relativeSpaceBone {boneName=root}
   }
   n34 = Anim Dynamics - Bone: dyn_belt_flap_bck_l_02(ComponentPose=@n33, Alpha=1, ExternalForce="0,0,0") {
    angularSpringConstant=0.3
    bAngularSpring=true
    bGravityOverrideInSimSpace=true
    bUseGravityOverride=true
    simulationSpace=World
    alphaBoolBlend {blendOption=Linear}
    boundBone {boneName=dyn_belt_flap_bck_l_02}
    componentAppliedLinearAccClamp=<10000 10000 10000>
    componentLinearAccScale=<0.5 0.5 0.5>
    componentLinearVelScale=<0.5 0.5 0.5>
    gravityOverride {z=-1500}
    physicsBodyDefinitions [{
      boundBone {boneName=dyn_belt_flap_bck_l_02}
      boxExtents=<16 16 16>
      constraintSetup {angularLimitsMax=<0 2 1>, angularLimitsMin=<0 -50 -1>, angularTarget=<0 0 0>}
      localJointOffset=<1 0 0>
     }]
    relativeSpaceBone {boneName=root}
   }
   n35 = Anim Dynamics - Bone: dyn_belt_flap_bck_l_03(ComponentPose=@n34, Alpha=1, ExternalForce="0,0,0") {
    angularSpringConstant=0.3
    bAngularSpring=true
    bGravityOverrideInSimSpace=true
    bUseGravityOverride=true
    simulationSpace=World
    alphaBoolBlend {blendOption=Linear}
    boundBone {boneName=dyn_belt_flap_bck_l_03}
    componentAppliedLinearAccClamp=<10000 10000 10000>
    componentLinearAccScale=<0.5 0.5 0.5>
    componentLinearVelScale=<0.5 0.5 0.5>
    gravityOverride {z=-1500}
    physicsBodyDefinitions [{
      boundBone {boneName=dyn_belt_flap_bck_l_03}
      boxExtents=<14 14 14>
      constraintSetup {angularLimitsMax=<0 0 1>, angularLimitsMin=<0 -50 -1>, angularTarget=<0 0 0>}
      localJointOffset=<1 0 0>
     }]
    relativeSpaceBone {boneName=root}
   }
   n36 = Anim Dynamics - Bone: dyn_belt_flap_bck_l_04(ComponentPose=@n35, Alpha=1, ExternalForce="0,0,0") {
    angularSpringConstant=0.5
    bAngularSpring=true
    bGravityOverrideInSimSpace=true
    bUseGravityOverride=true
    simulationSpace=World
    alphaBoolBlend {blendOption=Linear}
    boundBone {boneName=dyn_belt_flap_bck_l_04}
    componentAppliedLinearAccClamp=<10000 10000 10000>
    componentLinearAccScale=<0.5 0.5 0.5>
    componentLinearVelScale=<0.5 0.5 0.5>
    gravityOverride {z=-1500}
    physicsBodyDefinitions [{
      boundBone {boneName=dyn_belt_flap_bck_l_04}
      boxExtents=<14 14 14>
      constraintSetup {angularLimitsMax=<0 2 1>, angularLimitsMin=<0 -50 -1>, angularTarget=<0 0 0>}
      localJointOffset=<1 0 0>
     }]
    relativeSpaceBone {boneName=root}
   }
   n37 = Anim Dynamics - Bone: dyn_belt_flap_bck_r_01(ComponentPose=@n36, Alpha=1, ExternalForce="0,0,0") {
    angularSpringConstant=5000
    bAngularSpring=true
    bGravityOverrideInSimSpace=true
    bUseGravityOverride=true
    simulationSpace=World
    alphaBoolBlend {blendOption=Linear}
    boundBone {boneName=dyn_belt_flap_bck_r_01}
    componentAppliedLinearAccClamp=<10000 10000 10000>
    componentLinearAccScale=<0.5 0.5 0.5>
    componentLinearVelScale=<0.5 0.5 0.5>
    gravityOverride {z=-1500}
    physicsBodyDefinitions [{
      boundBone {boneName=dyn_belt_flap_bck_r_01}
      boxExtents=<16 16 16>
      constraintSetup {angularLimitsMax=<0 0 1>, angularLimitsMin=<0 -50 -1>, angularTarget=<0 0 0>}
      localJointOffset=<1 0 0>
     }]
    relativeSpaceBone {boneName=root}
   }
   n38 = Anim Dynamics - Bone: dyn_belt_flap_bck_r_02(ComponentPose=@n37, Alpha=1, ExternalForce="0,0,0") {
    angularSpringConstant=5000
    bAngularSpring=true
    bGravityOverrideInSimSpace=true
    bUseGravityOverride=true
    simulationSpace=World
    alphaBoolBlend {blendOption=Linear}
    boundBone {boneName=dyn_belt_flap_bck_r_02}
    componentAppliedLinearAccClamp=<10000 10000 10000>
    componentLinearAccScale=<0.5 0.5 0.5>
    componentLinearVelScale=<0.5 0.5 0.5>
    gravityOverride {z=-1500}
    physicsBodyDefinitions [{
      boundBone {boneName=dyn_belt_flap_bck_r_02}
      boxExtents=<16 16 16>
      constraintSetup {angularLimitsMax=<0 2 1>, angularLimitsMin=<0 -50 -1>, angularTarget=<0 0 0>}
      localJointOffset=<1 0 0>
     }]
    relativeSpaceBone {boneName=root}
   }
   n39 = Anim Dynamics - Bone: dyn_belt_flap_bck_r_03(ComponentPose=@n38, Alpha=1, ExternalForce="0,0,0") {
    angularSpringConstant=3000
    bAngularSpring=true
    bGravityOverrideInSimSpace=true
    bUseGravityOverride=true
    simulationSpace=World
    alphaBoolBlend {blendOption=Linear}
    boundBone {boneName=dyn_belt_flap_bck_r_03}
    componentAppliedLinearAccClamp=<10000 10000 10000>
    componentLinearAccScale=<0.5 0.5 0.5>
    componentLinearVelScale=<0.5 0.5 0.5>
    gravityOverride {z=-1500}
    physicsBodyDefinitions [{
      boundBone {boneName=dyn_belt_flap_bck_r_03}
      boxExtents=<14 14 14>
      constraintSetup {angularLimitsMax=<0 0 1>, angularLimitsMin=<0 -50 -1>, angularTarget=<0 0 0>}
      localJointOffset=<1 0 0>
     }]
    relativeSpaceBone {boneName=root}
   }
   n40 = Anim Dynamics - Bone: dyn_belt_flap_bck_r_04(ComponentPose=@n39, Alpha=1, ExternalForce="0,0,0") {
    angularSpringConstant=3000
    bAngularSpring=true
    bGravityOverrideInSimSpace=true
    bUseGravityOverride=true
    simulationSpace=World
    alphaBoolBlend {blendOption=Linear}
    boundBone {boneName=dyn_belt_flap_bck_r_04}
    componentAppliedLinearAccClamp=<10000 10000 10000>
    componentLinearAccScale=<0.5 0.5 0.5>
    componentLinearVelScale=<0.5 0.5 0.5>
    gravityOverride {z=-1500}
    physicsBodyDefinitions [{
      boundBone {boneName=dyn_belt_flap_bck_r_04}
      boxExtents=<14 14 14>
      constraintSetup {angularLimitsMax=<0 2 1>, angularLimitsMin=<0 -50 -1>, angularTarget=<0 0 0>}
      localJointOffset=<1 0 0>
     }]
    relativeSpaceBone {boneName=root}
   }
   n41 = Anim Dynamics - Bone: dyn_shoulderpad_l_chain_01(ComponentPose=@n40, Alpha=1, ExternalForce="0,0,0") {
    angularSpringConstant=0.5
    bAngularSpring=true
    bGravityOverrideInSimSpace=true
    bUseGravityOverride=true
    simulationSpace=BoneRelative
    alphaBoolBlend {blendOption=Linear}
    boundBone {boneName=dyn_shoulderpad_l_chain_01}
    componentAppliedLinearAccClamp=<10000 10000 10000>
    componentLinearAccScale=<1 1 1>
    componentLinearVelScale=<1 1 1>
    gravityOverride {z=-1200}
    physicsBodyDefinitions [{
      boundBone {boneName=dyn_shoulderpad_l_chain_01}
      boxExtents=<17 17 17>
      constraintSetup {angularLimitsMax=<0 5 5>, angularLimitsMin=<0 -50 -5>, angularTarget=<0 0 0>}
      localJointOffset=<1 0 0>
     }]
    relativeSpaceBone {boneName=root}
   }
   n42 = Anim Dynamics - Bone: dyn_shoulderpad_l_chain_02(ComponentPose=@n41, Alpha=1, ExternalForce="0,0,0") {
    angularSpringConstant=0.3
    bAngularSpring=true
    bGravityOverrideInSimSpace=true
    bUseGravityOverride=true
    simulationSpace=BoneRelative
    alphaBoolBlend {blendOption=Linear}
    boundBone {boneName=dyn_shoulderpad_l_chain_02}
    componentAppliedLinearAccClamp=<10000 10000 10000>
    componentLinearAccScale=<1 1 1>
    componentLinearVelScale=<1 1 1>
    gravityOverride {z=-1200}
    physicsBodyDefinitions [{
      boundBone {boneName=dyn_shoulderpad_l_chain_02}
      boxExtents=<17 17 17>
      constraintSetup {angularLimitsMax=<0 3 5>, angularLimitsMin=<0 -50 -5>, angularTarget=<0 0 0>}
      localJointOffset=<1 0 0>
     }]
    relativeSpaceBone {boneName=root}
   }
   n43 = Anim Dynamics - Bone: dyn_shoulderpad_l_chain_01(ComponentPose=@n42, Alpha=1, ExternalForce="0,0,0") {
    angularSpringConstant=3000
    bAngularSpring=true
    bGravityOverrideInSimSpace=true
    bUseGravityOverride=true
    simulationSpace=World
    alphaBoolBlend {blendOption=Linear}
    boundBone {boneName=dyn_shoulderpad_l_chain_01}
    componentAppliedLinearAccClamp=<10000 10000 10000>
    componentLinearAccScale=<0.5 0.5 0.5>
    componentLinearVelScale=<0.5 0.5 0.5>
    gravityOverride {z=-1500}
    physicsBodyDefinitions [{
      boundBone {boneName=dyn_shoulderpad_l_chain_01}
      boxExtents=<14 14 14>
      constraintSetup {angularLimitsMax=<0 5 5>, angularLimitsMin=<0 -50 -5>, angularTarget=<0 0 0>}
      localJointOffset=<1 0 0>
     }]
    relativeSpaceBone {boneName=root}
   }
   n44 = Anim Dynamics - Bone: dyn_shoulderpad_l_chain_02(ComponentPose=@n43, Alpha=1, ExternalForce="0,0,0") {
    angularSpringConstant=0.3
    bAngularSpring=true
    bGravityOverrideInSimSpace=true
    bUseGravityOverride=true
    simulationSpace=World
    alphaBoolBlend {blendOption=Linear}
    boundBone {boneName=dyn_shoulderpad_l_chain_02}
    componentAppliedLinearAccClamp=<10000 10000 10000>
    componentLinearAccScale=<1 1 1>
    componentLinearVelScale=<1 1 1>
    gravityOverride {z=-1200}
    physicsBodyDefinitions [{
      boundBone {boneName=dyn_shoulderpad_l_chain_02}
      boxExtents=<14 14 14>
      constraintSetup {angularLimitsMax=<0 3 5>, angularLimitsMin=<0 -50 -5>, angularTarget=<0 0 0>}
      localJointOffset=<1 0 0>
     }]
    relativeSpaceBone {boneName=root}
   }
   Output Pose(Result=Component To Local(ComponentPose=@n44)) {layerGroup=None}
  ]}
 EventGraph {nodes [
   Event BlueprintUpdateAnimation()
   > n1 = Sequence()
   n2 = CheckForBackbling()
   Set Strap Linear Acc Scale(
    execute=@n1.then_1
    Strap Linear Acc Scale:Vector="To Vector (Float)"!(InDouble:double=SelectFloat(A=0.1, B=0.3, bPickA=@n2)) -> (ReturnValue:Vector)
   ) -> (Output_Get:Vector)
   > n3 = Branch(Condition=Get bIsInFrontEnd())
   > Set AnimDyn FE(AnimDyn FE:double=1) -> (Output_Get:double)
   > Set AnimDyn InGame(AnimDyn InGame:double=0) -> (Output_Get:double)
   Set AnimDyn FE(execute=@n3.else, AnimDyn FE=0)
   > Set AnimDyn InGame(AnimDyn InGame=1)
   n4 = Gravity Override State Function DEPRECATED(
    execute=@n1.then_0
    CharacterPartAnimInstance=Self-Reference()
    GameTimeInSeconds=GetGameTimeInSeconds()
    GravityOverrideParams=Get GravityOverrideParams() -> (GravityOverrideParams:Gravity Override Params Struct Structure)
   )
   > Set GravityOverride(GravityOverride=@n4)
  ]}
]
variables [
 AnimDyn FE {type=number}
 AnimDyn InGame {type=number}
 AnimDyn SimSpace {type=AnimPhysSimSpaceType}
 GravityOverride {type=Vector}
 GravityOverrideParams {type=GravityOverrideParamsStruct}
 Strap Linear Acc Scale {type=Vector}
 UpperArm_RotNeg_L {type=number}
 UpperArm_RotNeg_R {type=number}
]
