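# Physics asset for the male medium PiteousKicks player body, named as the collider set for its cape.
# Capsules cover pelvis, spine, neck and clavicles, with a sphere on the head; arm, leg and foot bodies carry no shapes.
# Dynamic bones for two four-link belt straps, two pistols and a sword add capsules.
# Fully free constraints chain the straps to the pelvis and link the pistols and sword to the spine.
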PhysicsAsset /Game/Characters/Player/Male/Medium/Bodies/M_MED_PiteousKicks/Meshes/M_MED_PiteousKicks_CapeColliders
bodies {
 calf_l []
 calf_r []
 clavicle_l [{shapeType=Capsule, length=7.5192, radius=11.3871, center=<14.7908 -0.7683 -2.9511>, rotation=<rot 69.3297 1.4476 176.9462>}]
 clavicle_r [{shapeType=Capsule, length=7.5192, radius=11.3871, center=<-11.7399 1.0247 1.5314>, rotation=<rot 69.3297 1.4476 -3.0538>}]
 dyn_belt_strap_l_01 [{shapeType=Capsule, length=5, radius=8, shapeName=dyn_belt_strap_l_01_capsule, center=<5.9562 -0.0803 0>, rotation=<rot 90 180 179.9998>}]
 dyn_belt_strap_l_02 [{shapeType=Capsule, length=5, radius=8, shapeName=dyn_belt_strap_l_02_capsule, center=<6.4041 -0.0803 0>, rotation=<rot 90 -0.0008 -0.001>}]
 dyn_belt_strap_l_03 [{
   shapeType=Capsule
   length=6.5
   radius=8
   shapeName=dyn_belt_strap_l_03_capsule
   center=<8.2641 -0.0803 0>
   rotation=<rot 90 -179.9999 179.9999>
  }]
 dyn_belt_strap_l_04 [{
   shapeType=Capsule
   length=26
   radius=10
   shapeName=dyn_belt_strap_l_04_capsule
   center=<16.656 -0.0803 0>
   rotation=<rot 90 -166.7373 -166.7376>
  }]
 dyn_belt_strap_r_01 [{shapeType=Capsule, length=5, radius=8, shapeName=dyn_belt_strap_r_01_capsule, center=<5.9562 -0.0804 0>, rotation=<rot 90 179.9802 179.9804>}]
 dyn_belt_strap_r_02 [{shapeType=Capsule, length=5, radius=8, shapeName=dyn_belt_strap_r_02_capsule, center=<6.4041 -0.0804 0>, rotation=<rot 90 178.7341 178.7343>}]
 dyn_belt_strap_r_03 [{shapeType=Capsule, length=6.5, radius=8, shapeName=dyn_belt_strap_r_03_capsule, center=<8.2641 -0.0804 0>, rotation=<rot 90 180 -179.9998>}]
 dyn_belt_strap_r_04 [{shapeType=Capsule, length=26, radius=10, shapeName=dyn_belt_strap_r_04_capsule, center=<16.656 -0.0804 0>, rotation=<rot 90 180 -179.9998>}]
 dyn_pistol_l [{
   shapeType=Capsule
   length=14
   radius=4.6
   shapeName=dyn_pistol_l_capsule
   center=<2.6247 -4.4079 -4.3023>
   rotation=<rot 0 -171.1398 -43.8521>
  }]
 dyn_pistol_r [{shapeType=Capsule, length=14, radius=4.6, shapeName=dyn_pistol_r_capsule, center=<2.6236 4.3906 -4.3202>, rotation=<rot 0 -8.8865 -43.6224>}]
 dyn_sword [{shapeType=Capsule, length=26, radius=5.5, shapeName=dyn_sword_capsule, center=<6.2474 -5.2509 -14.0155>, rotation=<rot 21.8836 -90 -14.0025>}]
 foot_l []
 foot_r []
 head [{shapeType=Sphere, radius=17.3125, center=<11.028 -0.174 0>}]
 lowerarm_l []
 lowerarm_r []
 neck_02 [{shapeType=Capsule, length=13.4375, radius=7.3175, center=<-0.2642 -0.3408 0>, rotation=<rot 90 -171.1193 -175.3978>}]
 pelvis [{shapeType=Capsule, length=9.6456, radius=17, center=<-6.31 -1.11 0>}]
 spine_01 [{shapeType=Capsule, length=7.9438, radius=13.5586, center=<-1.7965 0.1884 0>}]
 spine_02 [
  {shapeType=Capsule, length=7.3264, radius=13.6322, center=<-1.711 -0.1851 0.0001>}
  {shapeType=Capsule, length=29.3956, radius=11.5807, center=<2.5024 0.4992 3.9825>, rotation=<rot 90 0.0049 0.0049>}
  {shapeType=Capsule, length=29.875, radius=11.1013, center=<2.4704 0.4991 -3.9825>, rotation=<rot 90 -0.0396 -0.0396>}
 ]
 spine_03 [{shapeType=Capsule, length=8.2107, radius=13.7031, center=<0.1445 0.2396 0.0001>}]
 spine_04 [{shapeType=Capsule, length=9.4404, radius=14.1325, center=<2.766 0.7765 0.0002>}]
 spine_05 [{shapeType=Capsule, length=8.0226, radius=15.695, center=<-2 1 0>}]
 thigh_l []
 thigh_r []
 upperarm_l []
 upperarm_r []
}
constraints [
 {
  bone1Name=dyn_belt_strap_r_04
  bone2Name=dyn_belt_strap_r_03
  swing1LimitDegrees=45
  swing1Motion=Free
  swing2LimitDegrees=45
  swing2Motion=Free
  twistLimitDegrees=45
  twistMotion=Free
 }
 {
  bone1Name=dyn_belt_strap_r_03
  bone2Name=dyn_belt_strap_r_02
  swing1LimitDegrees=45
  swing1Motion=Free
  swing2LimitDegrees=45
  swing2Motion=Free
  twistLimitDegrees=45
  twistMotion=Free
 }
 {
  bone1Name=dyn_belt_strap_r_02
  bone2Name=dyn_belt_strap_r_01
  swing1LimitDegrees=45
  swing1Motion=Free
  swing2LimitDegrees=45
  swing2Motion=Free
  twistLimitDegrees=45
  twistMotion=Free
 }
 {
  bone1Name=dyn_belt_strap_r_01
  bone2Name=pelvis
  swing1LimitDegrees=45
  swing1Motion=Free
  swing2LimitDegrees=45
  swing2Motion=Free
  twistLimitDegrees=45
  twistMotion=Free
 }
 {
  bone1Name=dyn_belt_strap_l_04
  bone2Name=dyn_belt_strap_l_03
  swing1LimitDegrees=45
  swing1Motion=Free
  swing2LimitDegrees=45
  swing2Motion=Free
  twistLimitDegrees=45
  twistMotion=Free
 }
 {
  bone1Name=dyn_belt_strap_l_03
  bone2Name=dyn_belt_strap_l_02
  swing1LimitDegrees=45
  swing1Motion=Free
  swing2LimitDegrees=45
  swing2Motion=Free
  twistLimitDegrees=45
  twistMotion=Free
 }
 {
  bone1Name=dyn_belt_strap_l_02
  bone2Name=dyn_belt_strap_l_01
  swing1LimitDegrees=45
  swing1Motion=Free
  swing2LimitDegrees=45
  swing2Motion=Free
  twistLimitDegrees=45
  twistMotion=Free
 }
 {
  bone1Name=dyn_belt_strap_l_01
  bone2Name=pelvis
  swing1LimitDegrees=45
  swing1Motion=Free
  swing2LimitDegrees=45
  swing2Motion=Free
  twistLimitDegrees=45
  twistMotion=Free
 }
 {
  bone1Name=dyn_sword
  bone2Name=spine_01
  swing1LimitDegrees=45
  swing1Motion=Free
  swing2LimitDegrees=45
  swing2Motion=Free
  twistLimitDegrees=45
  twistMotion=Free
 }
 {
  bone1Name=dyn_pistol_r
  bone2Name=spine_04
  swing1LimitDegrees=45
  swing1Motion=Free
  swing2LimitDegrees=45
  swing2Motion=Free
  twistLimitDegrees=45
  twistMotion=Free
 }
 {
  bone1Name=dyn_pistol_l
  bone2Name=spine_04
  swing1LimitDegrees=45
  swing1Motion=Free
  swing2LimitDegrees=45
  swing2Motion=Free
  twistLimitDegrees=45
  twistMotion=Free
 }
]
